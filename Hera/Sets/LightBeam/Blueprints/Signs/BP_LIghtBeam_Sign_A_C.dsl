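; BP_LIghtBeam_Sign_A_C is a LightBeam set sign Blueprint that passes its dynamic sign Number to the static mesh's material.
; ChangeNumbers writes Number into default custom primitive data float index 0, so the value can vary per instance without a dynamic material instance.
; It runs from the construction script and again on BeginPlay after the parent's logic.
; The Tick and overlap events are empty.
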(event EventBeginPlay
  (|CallParentFunction)
  (CallFunction|ChangeNumbers))

(event Collision|EventActorBeginOverlap
  (params (OtherActor :type Actor%20Object%20Reference)))

(event EventTick
  (params (DeltaSeconds :type Float%20%28single-precision%29)))

(fn ChangeNumbers
  (Rendering|Material|SetDefaultCustomPrimitiveDataFloat (Variables|Debug|GetStaticMeshComponent) 0 (Variables|DynamicSign|GetNumber)))

(fn ConstructionScript
  (CallFunction|ChangeNumbers))
